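# Parent material for Trix vehicle wheel cosmetics, adding colorable, pulsing emissive to rim and tire.
# MF_WheelParent_Simple supplies the base attributes; vertex-color red masks blend in emissive before damage and de-rez transition functions.
# RGB masks combine accent and custom colors with brightness parameters and a time-driven sine pulse, which static switches can replace with constants on the tire.
# A Fresnel glow branch using Trix Radiant masks ends unconnected.
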Material /VehicleCosmetics/Wheels/Trix/Materials/M_WheelParent_Trix
graph {
 properties {
  bAutomaticallySetUsageInEditor=false
  bUseMaterialAttributes=true
  bUsedWithMeshDeformer=true
  bUsedWithSkeletalMesh=true
  pixelDepthOffsetMode=PDOM_Legacy
  previewMesh=/VehicleCosmetics/Wheels/Trix/Mesh/SK_Wheel_Trix
  usedShadingModels=Default Lit
 }
 nodes [
  n1 = Sine(None=Multiply(A=Time(), B=Constant() {r=2})) {period=2}
  n2 = Constant() {r=1}
  Add(A=@n1, B=@n2)
  n3 = NamedRerouteUsage()
  n4 = Constant() {r=1.1}
  Add(A=@n3, B=@n4)
  CameraVectorWS()
  Constant()
  n5 = MaterialFunctionCall() {materialFunction=&MF_WheelParent_Simple}
  n6 = GetMaterialAttributes(None=@n5.MaterialAttributes) {attributeGetTypes [#1]}
  n7 = VertexColor()
  n8 = LinearInterpolate(A=LinearInterpolate(A=@n6.Emissive Color, B=NamedRerouteUsage(), Alpha=@n7.R), B=NamedRerouteUsage(), Alpha=OneMinus(None=@n5.VertexColorR))
  MaterialFunctionCall(In=MaterialFunctionCall(Material=SetMaterialAttributes(MaterialAttributes=@n6.MaterialAttributes, Emissive Color=@n8) {attributeSetTypes [#1]}) {materialFunction=&MF_BaseVehicle_Damage}.Result) {materialFunction=&MF_VehicleDeRez_Transition}
  Multiply(A=@n3, B=@n4)
  Multiply()
  n9 = Time()
  n10 = Constant() {r=2.5}
  n11 = Add(A=Sine(None=Multiply(A=@n9, B=@n10)), B=Constant() {r=1.2})
  n12 = Power(
   Base=ComponentMask(None=MaterialFunctionCall!(Input Normal, Input Sharpness, Input Two Sided) {materialFunction=/VehicleCosmetics/Wheels/CarCar_Holographic/Materials/MF_Fresnel_Advanced}.Middle) {b=true}
   Exp=Constant() {r=1.2}
  )
  n13 = Clamp(None=ConstantBiasScale(None=Clamp(None=@n12)) {bias=0, scale=0.25}) {maxDefault=1.5}
  n14 = TextureSampleParameter2D() {parameterName=Masks, samplerType=SAMPLERTYPE_LinearColor, texture=/VehicleCosmetics/Wheels/Trix_Radiant/Textures/T_Trix_Radiant_Rim_M}
  n15 = Multiply(A=@n14.G, B=@n11)
  n16 = Multiply(A=@n14.R, B=@n11)
  n17 = VectorParameter() {group=5_Special Effects, parameterName=AccentColor, defaultValue=<rgba 0.9734 0 0.6867 1>}
  n18 = VectorParameter() {group=5_Special Effects, parameterName=CustomColor, defaultValue=<rgba 0 1 0.1329 1>}
  n19 = Constant() {r=0.25}
  Multiply(
   A=Multiply(A=Multiply(A=@n11, B=@n13), B=Add(A=@n15, B=@n16))
   B=LinearInterpolate(A=@n17.RGB, B=Multiply(A=@n18.RGB, B=@n19), Alpha=NamedRerouteUsage())
  )
  n20 = TextureSampleParameter2D() {
   group=5_Special Effects
   parameterName=Rim_Emissive_M
   samplerSource=SSM_Wrap_WorldGroupSettings
   samplerType=SAMPLERTYPE_LinearColor
   texture=/VehicleCosmetics/Wheels/Trix/Textures/T_Trix_Rim_RGB
  }
  n21 = NamedRerouteUsage()
  n22 = NamedRerouteUsage()
  n23 = Add(A=Multiply(A=Multiply(A=@n20.G, B=@n17.RGB), B=@n21), B=Multiply(A=@n20.G, B=Multiply(A=@n17.RGB, B=@n22)))
  n24 = Clamp(None=Multiply(A=@n1, B=@n2))
  n25 = Multiply(A=@n24, B=@n18.RGB)
  n26 = Add(A=Multiply(A=Multiply(A=@n20.R, B=@n18.RGB), B=@n21), B=Multiply(A=@n20.R, B=Multiply(A=@n20.R, B=@n25)))
  n27 = Multiply(
   A=@n20.B
   B=Multiply(A=Add(A=@n23, B=@n26), B=ScalarParameter() {defaultValue=2, group=5_Special Effects, parameterName=RimEmissiveBrughtness})
  )
  n28 = TextureSampleParameter2D() {
   group=5_Special Effects
   parameterName=TireEmissive_M
   samplerSource=SSM_Wrap_WorldGroupSettings
   samplerType=SAMPLERTYPE_LinearColor
   texture=/VehicleCosmetics/Wheels/Trix/Textures/T_Trix_Tire_RGB
  }
  n29 = ScalarParameter() {group=5_Special Effects, parameterName=ConstantEmissive}
  n30 = Multiply(
   A=Multiply(A=VectorParameter() {group=5_Special Effects, parameterName=TireAccentColor, defaultValue=<rgba 1 0 0.552 1>}.RGB, B=@n28.G)
   B=@n29
  )
  n31 = Constant() {r=0.15}
  n32 = Multiply(A=Constant() {r=3}, B=StaticSwitchParameter(True=@n31, False=@n24) {group=5_Special Effects})
  n33 = Multiply(A=@n32, B=@n28.G)
  n34 = Multiply(A=@n29, B=Multiply(A=@n18.RGB, B=@n28.R))
  n35 = Multiply(A=StaticSwitchParameter(True=@n31, False=@n24) {group=5_Special Effects}, B=Constant() {r=1})
  n36 = Multiply(A=Multiply(A=@n18.RGB, B=@n35), B=@n28.R)
  n37 = Multiply(
   A=Add(A=Add(A=@n30, B=@n33), B=Add(A=@n34, B=@n36))
   B=ScalarParameter() {defaultValue=2, group=5_Special Effects, parameterName=TireEmissiveBrighntess}
  )
  NamedRerouteDeclaration(None=LinearInterpolate(A=@n27, B=Multiply(A=@n37, B=@n28.B), Alpha=@n7.R)) {name=Emissive_Tire, nodeColor=<rgba 1 0.75 0 1>}
  NamedRerouteDeclaration(None=@n14.R) {name=Red_CM, nodeColor=<rgba 1 0.3047 0 1>}
  NamedRerouteDeclaration(None=@n14.G) {name=Green_CM, nodeColor=<rgba 1 0.3047 0 1>}
  NamedRerouteDeclaration(None=@n29) {name=ConstantEmissive, nodeColor=<rgba 0.0547 0 1 1>}
  NamedRerouteDeclaration(None=@n32) {name=Animatian1, nodeColor=<rgba 0.2656 1 0 1>}
  NamedRerouteDeclaration(None=@n35) {name=Animation2, nodeColor=<rgba 1 0 0.5859 1>}
  NamedRerouteDeclaration(None=@n27) {name=Emissive_Rim, nodeColor=<rgba 1 0.75 0 1>}
 ]
}
